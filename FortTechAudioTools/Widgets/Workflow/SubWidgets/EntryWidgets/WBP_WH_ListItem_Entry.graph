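# List-entry widget for FortTechAudioTools workflow UI, showing a class thumbnail, an action button and read-only text.
# On item set, Initialize casts the item to this same widget class and copies its WorkflowEntryData to configure the button and text.
# It loads the asset class synchronously for the thumbnail, showing a dash if invalid, and hides the text area when empty.
# Button presses broadcast OnButtonPressed with the entry data.
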WidgetBlueprint /FortTechAudioTools/Widgets/Workflow/SubWidgets/EntryWidgets/WBP_WH_ListItem_Entry
graphs [
 EventGraph {nodes [
   "On Pressed (Button)"() {is=title K2Node_ComponentBoundEvent}
   > Call On Button Pressed!(
    execute:Exec
    self:WBP WH List Item Entry Object Reference
    WorkflowEntryData:Struct WH Entry Widget Data Structure=Get WorkflowEntryData() -> (WorkflowEntryData:Struct WH Entry Widget Data Structure)
   ) -> (then:Exec) {is=title K2Node_CallDelegate}
   n1 = Event OnListItemObjectSet()
   > Initialize!(execute, self, Object:Object Reference=@n1.ListItemObject) -> (then)
   Event PreConstruct()
   Event Construct()
   Event Tick()
  ]}
 Initialize {nodes [
   n1 = Initialize!() -> (then, Object:Object Reference) {is=title K2Node_FunctionEntry}
   > n2 = Cast To WBP_WH_ListItem_Entry!(execute, Object=@n1) -> (then, CastFailed, AsWBP WH List Item Entry:WBP WH List Item Entry Object Reference) {is=title K2Node_DynamicCast}
   > n3 = Set WorkflowEntryData(WorkflowEntryData:Struct WH Entry Widget Data Structure=Get WorkflowEntryData!(self:WBP WH List Item Entry Object Reference=@n2) -> (WorkflowEntryData)) -> (Output_Get:Struct WH Entry Widget Data Structure)
   n4 = Get Button() -> (Button:Editor Utility Button Object Reference)
   n5 = Break Struct WH Entry Widget Data!(Struct_WH_EntryWidget_Data:"Struct WH Entry Widget Data Structure (by ref)"=@n3) -> (
    bEnableButton_5_#1:bool=~
    ButtonText_2_#2:Text
    AssetClass_23_#3:Object Soft Class Reference
    Text_16_#4:Text
    TextTooltip_18_#5:Text
   ) {is=title K2Node_BreakStruct}
   > Set bIsEnabled!(execute, bIsEnabled:bool=@n5, self:Widget Object Reference=@n4) -> (then, Output_Get:bool)
   > Set Text(Text=@n5.ButtonText_2_#2, self=Get CommonTextBlock_36() -> (CommonTextBlock_36:Common Text Object Reference))
   > Set ToolTipText!(execute, ToolTipText:Text=@n5.TextTooltip_18_#5, self:Widget Object Reference=@n4) -> (then, Output_Get:Text)
   > n6 = LoadClassAsset_Blocking(AssetClass=@n5.AssetClass_23_#3)
   > n7 = Branch(Condition=IsValidClass!(Class:Object Class Reference=@n6) -> (ReturnValue:bool))
   n8 = GetSoftClassPath!(Class:Object Class Reference=@n6) -> (ReturnValue:Soft Class Path Structure)
   > n9 = SetAssetByObject!(
    execute
    self:Asset Thumbnail Widget Object Reference=Get Thumbnail() -> (Thumbnail:Asset Thumbnail Widget Object Reference)
    Object:Object Reference=Resolve Soft Reference!(Input:Soft Object Reference=To Soft Object Reference(SoftObjectPath=@n8)) -> (Output:Object Reference)
   ) -> (then)
   n10 = Branch(execute=[@n9.then, @n12.then], Condition=TextIsEmpty(InText=@n5.Text_16_#4))
   > SetVisibility(self=Get TextBorder() -> (TextBorder:Border Object Reference), InVisibility=Hidden)
   Set Text(
    execute=@n10.else
    Text=@n5.Text_16_#4
    self:Text Box Object Reference=Get ContentText() -> (ContentText:Text Box Object Reference)
   )
   SetVisibility(execute=@n7.else, self=Get Thumbnail(), InVisibility=Hidden)
   n11 = Get IconXText() -> (IconXText:Common Text Object Reference)
   > SetVisibility(self=@n11, InVisibility=Visible)
   > n12 = Set ToolTipText(ToolTipText=MakeLiteralText(Value="LOCTEXT(\"No Associated Class\")"), self=@n11)
  ]}
 OnButtonPressed {nodes [OnButtonPressed!() -> (then, WorkflowEntryData:Struct WH Entry Widget Data Structure) {is=title K2Node_FunctionEntry}]}
]
variables [
 OnButtonPressed {}
 WorkflowEntryData {type=Struct_WH_EntryWidget_Data, default {assetClass=null, bEnableButton=true, buttonText=Activate, text="", textTooltip=""}}
]
widgets {
 info {inheritedWidgetCount=0, parentClass=&UserWidget, rootWidgetClass=&HorizontalBox, widgetCount=14}
 widgets [
  HorizontalBox_302 {widgetClassPath=&HorizontalBox, widget=HorizontalBox_302}
  SizeBox_0 {
   widgetClassPath=&SizeBox
   parent=HorizontalBox_302
   widget=SizeBox_0
   properties {bOverride_HeightOverride=true, bOverride_WidthOverride=true, heightOverride=30, widthOverride=30}
   slot {class=HorizontalBoxSlot}
  }
  ThumbnailBorder {
   widgetClassPath=&Border
   bIsVariable=true
   parent=SizeBox_0
   widget=ThumbnailBorder
   properties {
    background {
     bIsDynamicallyLoaded=false
     drawAs=Image
     imageType=NoImage
     mirroring=NoMirror
     resourceName=None
     resourceObject=null
     tiling=NoTile
     imageSize=<32 32>
     margin=<ltrb 0 0 0 0>
     outlineSettings {
      bUseBrushTransparency=false
      roundingType=HalfHeightRadius
      width=0
      color {colorUseRule=UseColor_Specified, specifiedColor=<rgba 0 0 0 0>}
      cornerRadii=<0 0 0 0>
     }
     tintColor {colorUseRule=UseColor_Specified, specifiedColor=<rgba 0.0103 0.0103 0.0103 1>}
     uVRegion {bIsValid=false, max=<0 0>, min=<0 0>}
    }
    padding=<ltrb 0 0 0 0>
   }
   slot {class=SizeBoxSlot}
  }
  CanvasPanel_0 {
   widgetClassPath=&CanvasPanel
   parent=ThumbnailBorder
   widget=CanvasPanel_0
   slot {class=BorderSlot, properties {padding=<ltrb 0 0 0 0>}}
  }
  Thumbnail {
   widgetClassPath=/Script/UMGEditor.AssetThumbnailWidget
   bIsVariable=true
   parent=CanvasPanel_0
   widget=Thumbnail
   slot {class=CanvasPanelSlot, properties {layoutData {alignment=<0 0>, anchors {maximum=<1 1>, minimum=<0 0>}, offsets=<ltrb 0 0 0 0>}}}
  }
  IconXText {
   widgetClassPath=&CommonTextBlock
   bIsVariable=true
   parent=CanvasPanel_0
   widget=IconXText
   properties {
    style=/FortUILibrary/Styles/Typography/Foundation/TextStyle-Base-M_B_Grey.TextStyle-Base-M_B_Grey_C
    text="—"
    visibility=Hidden
    colorAndOpacity {colorUseRule=UseColor_Specified, specifiedColor=<rgba 0.4 0.4 0.4 1>}
    font {
     bForceMonospaced=false
     bMaterialIsStencil=false
     fontMaterial=null
     fontObject=&NotoSans
     letterSpacing=0
     monospacedWidth=1
     size=18
     skewAmount=0
     typefaceFontName=NotoSans-Bold
     outlineSettings {
      bApplyOutlineToDropShadows=false
      bMiteredCorners=false
      bSeparateFillAlpha=false
      outlineMaterial=null
      outlineSize=0
      outlineColor=<rgba 0 0 0 1>
     }
    }
    shadowOffset=<0 0>
   }
   slot {class=CanvasPanelSlot, properties {layoutData {alignment=<0 0>, anchors {maximum=<0 0>, minimum=<0 0>}, offsets=<ltrb 3.5 -4 30 30>}}}
  }
  Border_80 {
   widgetClassPath=&Border
   parent=HorizontalBox_302
   widget=Border_80
   properties {
    background {
     bIsDynamicallyLoaded=false
     drawAs=Image
     imageType=NoImage
     mirroring=NoMirror
     resourceName=None
     resourceObject=null
     tiling=NoTile
     imageSize=<32 32>
     margin=<ltrb 0 0 0 0>
     outlineSettings {
      bUseBrushTransparency=false
      roundingType=HalfHeightRadius
      width=0
      color {colorUseRule=UseColor_Specified, specifiedColor=<rgba 0 0 0 0>}
      cornerRadii=<0 0 0 0>
     }
     tintColor {colorUseRule=UseColor_Specified, specifiedColor=<rgba 1 1 1 0>}
     uVRegion {bIsValid=false, max=<0 0>, min=<0 0>}
    }
    padding=<ltrb 5 0 5 0>
   }
   slot {class=HorizontalBoxSlot}
  }
  SizeBox_59 {
   widgetClassPath=&SizeBox
   parent=Border_80
   widget=SizeBox_59
   properties {bOverride_HeightOverride=true, bOverride_WidthOverride=true, heightOverride=30, widthOverride=150}
   slot {class=BorderSlot, properties {padding=<ltrb 5 0 5 0>}}
  }
  Button {widgetClassPath=&EditorUtilityButton, parent=SizeBox_59, widget=Button, slot {class=SizeBoxSlot}}
  ScaleBox_0 {widgetClassPath=&ScaleBox, parent=Button, widget=ScaleBox_0, slot {class=ButtonSlot}}
  CommonTextBlock_36 {
   widgetClassPath=&CommonTextBlock
   bIsVariable=true
   parent=ScaleBox_0
   widget=CommonTextBlock_36
   properties {
    style=/FortUILibrary/Styles/Typography/Foundation/TextStyle-Base-XS.TextStyle-Base-XS_C
    text=Activate
    font {
     bForceMonospaced=false
     bMaterialIsStencil=false
     fontMaterial=null
     fontObject=&NotoSans
     letterSpacing=0
     monospacedWidth=1
     size=16
     skewAmount=0
     typefaceFontName=NotoSans-Regular
     outlineSettings {
      bApplyOutlineToDropShadows=false
      bMiteredCorners=false
      bSeparateFillAlpha=false
      outlineMaterial=null
      outlineSize=0
      outlineColor=<rgba 0 0 0 1>
     }
    }
    shadowOffset=<0 0>
   }
   slot {class=ScaleBoxSlot}
  }
  TextBorder {
   widgetClassPath=&Border
   bIsVariable=true
   parent=HorizontalBox_302
   widget=TextBorder
   properties {
    background {
     bIsDynamicallyLoaded=false
     drawAs=Image
     imageType=NoImage
     mirroring=NoMirror
     resourceName=None
     resourceObject=null
     tiling=NoTile
     imageSize=<32 32>
     margin=<ltrb 0 0 0 0>
     outlineSettings {
      bUseBrushTransparency=false
      roundingType=HalfHeightRadius
      width=0
      color {colorUseRule=UseColor_Specified, specifiedColor=<rgba 0 0 0 0>}
      cornerRadii=<0 0 0 0>
     }
     tintColor {colorUseRule=UseColor_Specified, specifiedColor=<rgba 0 0 0 1>}
     uVRegion {bIsValid=false, max=<0 0>, min=<0 0>}
    }
    padding=<ltrb 0 0 0 0>
   }
   slot {class=HorizontalBoxSlot, properties {size {sizeRule=Fill, value=1}}}
  }
  SizeBox_1 {
   widgetClassPath=&SizeBox
   parent=TextBorder
   widget=SizeBox_1
   properties {bOverride_HeightOverride=true, heightOverride=30}
   slot {class=BorderSlot, properties {padding=<ltrb 0 0 0 0>}}
  }
  ContentText {
   widgetClassPath=&EditableTextBox
   bIsVariable=true
   parent=SizeBox_1
   widget=ContentText
   properties {
    isReadOnly=true
    text=Text
    widgetStyle {
     backgroundColor {colorUseRule=UseColor_Specified, specifiedColor=<rgba 1 1 1 1>}
     backgroundImageFocused {
      bIsDynamicallyLoaded=false
      drawAs=Box
      imageType=FullColor
      mirroring=NoMirror
      resourceName=../../../Engine/Content/Slate/Common/TextBox_Hovered.png
      resourceObject=null
      tiling=NoTile
      imageSize=<0 0>
      margin=<ltrb 0.25 0.25 0.25 0.25>
      outlineSettings {
       bUseBrushTransparency=false
       roundingType=HalfHeightRadius
       width=0
       color {colorUseRule=UseColor_Specified, specifiedColor=<rgba 0 0 0 0>}
       cornerRadii=<0 0 0 0>
      }
      tintColor {colorUseRule=UseColor_Specified, specifiedColor=<rgba 1 1 1 1>}
      uVRegion {bIsValid=false, max=<0 0>, min=<0 0>}
     }
     backgroundImageHovered {
      bIsDynamicallyLoaded=false
      drawAs=Box
      imageType=FullColor
      mirroring=NoMirror
      resourceName=../../../Engine/Content/Slate/Common/TextBox_Hovered.png
      resourceObject=null
      tiling=NoTile
      imageSize=<0 0>
      margin=<ltrb 0.25 0.25 0.25 0.25>
      outlineSettings {
       bUseBrushTransparency=false
       roundingType=HalfHeightRadius
       width=0
       color {colorUseRule=UseColor_Specified, specifiedColor=<rgba 0 0 0 0>}
       cornerRadii=<0 0 0 0>
      }
      tintColor {colorUseRule=UseColor_Specified, specifiedColor=<rgba 1 1 1 1>}
      uVRegion {bIsValid=false, max=<0 0>, min=<0 0>}
     }
     backgroundImageNormal {
      bIsDynamicallyLoaded=false
      drawAs=Box
      imageType=FullColor
      mirroring=NoMirror
      resourceName=../../../Engine/Content/Slate/Common/TextBox.png
      resourceObject=null
      tiling=NoTile
      imageSize=<0 0>
      margin=<ltrb 0.25 0.25 0.25 0.25>
      outlineSettings {
       bUseBrushTransparency=false
       roundingType=HalfHeightRadius
       width=0
       color {colorUseRule=UseColor_Specified, specifiedColor=<rgba 0 0 0 0>}
       cornerRadii=<0 0 0 0>
      }
      tintColor {colorUseRule=UseColor_Specified, specifiedColor=<rgba 1 1 1 1>}
      uVRegion {bIsValid=false, max=<0 0>, min=<0 0>}
     }
     backgroundImageReadOnly {
      bIsDynamicallyLoaded=false
      drawAs=Box
      imageType=FullColor
      mirroring=NoMirror
      resourceName=../../../Engine/Content/Slate/Common/TextBox_ReadOnly.png
      resourceObject=null
      tiling=NoTile
      imageSize=<0 0>
      margin=<ltrb 0.25 0.25 0.25 0.25>
      outlineSettings {
       bUseBrushTransparency=false
       roundingType=HalfHeightRadius
       width=0
       color {colorUseRule=UseColor_Specified, specifiedColor=<rgba 0 0 0 0>}
       cornerRadii=<0 0 0 0>
      }
      tintColor {colorUseRule=UseColor_Specified, specifiedColor=<rgba 1 1 1 0>}
      uVRegion {bIsValid=false, max=<0 0>, min=<0 0>}
     }
     focusedForegroundColor {colorUseRule=UseColor_Foreground, specifiedColor=<rgba 1 0 1 1>}
     foregroundColor {colorUseRule=UseColor_Foreground, specifiedColor=<rgba 1 0 1 1>}
     hScrollBarPadding=<ltrb 0 0 0 0>
     padding=<ltrb 4 2 4 2>
     readOnlyForegroundColor {colorUseRule=UseColor_Foreground, specifiedColor=<rgba 1 0 1 1>}
     scrollBarStyle {
      thickness=16
      draggedThumbImage {
       bIsDynamicallyLoaded=false
       drawAs=Box
       imageType=FullColor
       mirroring=NoMirror
       resourceName=../../../Engine/Content/Slate/Common/Scrollbar_Thumb.png
       resourceObject=null
       tiling=NoTile
       imageSize=<0 0>
       margin=<ltrb 0.25 0.25 0.25 0.25>
       outlineSettings {
        bUseBrushTransparency=false
        roundingType=HalfHeightRadius
        width=0
        color {colorUseRule=UseColor_Specified, specifiedColor=<rgba 0 0 0 0>}
        cornerRadii=<0 0 0 0>
       }
       tintColor {colorUseRule=UseColor_Specified, specifiedColor=<rgba 1 1 1 1>}
       uVRegion {bIsValid=false, max=<0 0>, min=<0 0>}
      }
      horizontalBackgroundImage {
       bIsDynamicallyLoaded=false
       drawAs=NoDrawType
       imageType=NoImage
       mirroring=NoMirror
       resourceName=None
       resourceObject=null
       tiling=NoTile
       imageSize=<0 0>
       margin=<ltrb 0 0 0 0>
       outlineSettings {
        bUseBrushTransparency=false
        roundingType=HalfHeightRadius
        width=0
        color {colorUseRule=UseColor_Specified, specifiedColor=<rgba 0 0 0 0>}
        cornerRadii=<0 0 0 0>
       }
       tintColor {colorUseRule=UseColor_Specified, specifiedColor=<rgba 1 1 1 1>}
       uVRegion {bIsValid=false, max=<0 0>, min=<0 0>}
      }
      horizontalBottomSlotImage {
       bIsDynamicallyLoaded=false
       drawAs=Image
       imageType=FullColor
       mirroring=NoMirror
       resourceName=../../../Engine/Content/Slate/Common/Scrollbar_Background_Horizontal.png
       resourceObject=null
       tiling=NoTile
       imageSize=<8 8>
       margin=<ltrb 0 0 0 0>
       outlineSettings {
        bUseBrushTransparency=false
        roundingType=HalfHeightRadius
        width=0
        color {colorUseRule=UseColor_Specified, specifiedColor=<rgba 0 0 0 0>}
        cornerRadii=<0 0 0 0>
       }
       tintColor {colorUseRule=UseColor_Specified, specifiedColor=<rgba 1 1 1 1>}
       uVRegion {bIsValid=false, max=<0 0>, min=<0 0>}
      }
      horizontalTopSlotImage {
       bIsDynamicallyLoaded=false
       drawAs=Image
       imageType=FullColor
       mirroring=NoMirror
       resourceName=../../../Engine/Content/Slate/Common/Scrollbar_Background_Horizontal.png
       resourceObject=null
       tiling=NoTile
       imageSize=<8 8>
       margin=<ltrb 0 0 0 0>
       outlineSettings {
        bUseBrushTransparency=false
        roundingType=HalfHeightRadius
        width=0
        color {colorUseRule=UseColor_Specified, specifiedColor=<rgba 0 0 0 0>}
        cornerRadii=<0 0 0 0>
       }
       tintColor {colorUseRule=UseColor_Specified, specifiedColor=<rgba 1 1 1 1>}
       uVRegion {bIsValid=false, max=<0 0>, min=<0 0>}
      }
      hoveredThumbImage {
       bIsDynamicallyLoaded=false
       drawAs=Box
       imageType=FullColor
       mirroring=NoMirror
       resourceName=../../../Engine/Content/Slate/Common/Scrollbar_Thumb.png
       resourceObject=null
       tiling=NoTile
       imageSize=<0 0>
       margin=<ltrb 0.25 0.25 0.25 0.25>
       outlineSettings {
        bUseBrushTransparency=false
        roundingType=HalfHeightRadius
        width=0
        color {colorUseRule=UseColor_Specified, specifiedColor=<rgba 0 0 0 0>}
        cornerRadii=<0 0 0 0>
       }
       tintColor {colorUseRule=UseColor_Specified, specifiedColor=<rgba 1 1 1 1>}
       uVRegion {bIsValid=false, max=<0 0>, min=<0 0>}
      }
      normalThumbImage {
       bIsDynamicallyLoaded=false
       drawAs=Box
       imageType=FullColor
       mirroring=NoMirror
       resourceName=../../../Engine/Content/Slate/Common/Scrollbar_Thumb.png
       resourceObject=null
       tiling=NoTile
       imageSize=<0 0>
       margin=<ltrb 0.25 0.25 0.25 0.25>
       outlineSettings {
        bUseBrushTransparency=false
        roundingType=HalfHeightRadius
        width=0
        color {colorUseRule=UseColor_Specified, specifiedColor=<rgba 0 0 0 0>}
        cornerRadii=<0 0 0 0>
       }
       tintColor {colorUseRule=UseColor_Specified, specifiedColor=<rgba 1 1 1 1>}
       uVRegion {bIsValid=false, max=<0 0>, min=<0 0>}
      }
      verticalBackgroundImage {
       bIsDynamicallyLoaded=false
       drawAs=NoDrawType
       imageType=NoImage
       mirroring=NoMirror
       resourceName=None
       resourceObject=null
       tiling=NoTile
       imageSize=<0 0>
       margin=<ltrb 0 0 0 0>
       outlineSettings {
        bUseBrushTransparency=false
        roundingType=HalfHeightRadius
        width=0
        color {colorUseRule=UseColor_Specified, specifiedColor=<rgba 0 0 0 0>}
        cornerRadii=<0 0 0 0>
       }
       tintColor {colorUseRule=UseColor_Specified, specifiedColor=<rgba 1 1 1 1>}
       uVRegion {bIsValid=false, max=<0 0>, min=<0 0>}
      }
      verticalBottomSlotImage {
       bIsDynamicallyLoaded=false
       drawAs=Image
       imageType=FullColor
       mirroring=NoMirror
       resourceName=../../../Engine/Content/Slate/Common/Scrollbar_Background_Vertical.png
       resourceObject=null
       tiling=NoTile
       imageSize=<8 8>
       margin=<ltrb 0 0 0 0>
       outlineSettings {
        bUseBrushTransparency=false
        roundingType=HalfHeightRadius
        width=0
        color {colorUseRule=UseColor_Specified, specifiedColor=<rgba 0 0 0 0>}
        cornerRadii=<0 0 0 0>
       }
       tintColor {colorUseRule=UseColor_Specified, specifiedColor=<rgba 1 1 1 1>}
       uVRegion {bIsValid=false, max=<0 0>, min=<0 0>}
      }
      verticalTopSlotImage {
       bIsDynamicallyLoaded=false
       drawAs=Image
       imageType=FullColor
       mirroring=NoMirror
       resourceName=../../../Engine/Content/Slate/Common/Scrollbar_Background_Vertical.png
       resourceObject=null
       tiling=NoTile
       imageSize=<8 8>
       margin=<ltrb 0 0 0 0>
       outlineSettings {
        bUseBrushTransparency=false
        roundingType=HalfHeightRadius
        width=0
        color {colorUseRule=UseColor_Specified, specifiedColor=<rgba 0 0 0 0>}
        cornerRadii=<0 0 0 0>
       }
       tintColor {colorUseRule=UseColor_Specified, specifiedColor=<rgba 1 1 1 1>}
       uVRegion {bIsValid=false, max=<0 0>, min=<0 0>}
      }
     }
     textStyle {
      overflowPolicy=Clip
      transformPolicy=None
      colorAndOpacity {colorUseRule=UseColor_Foreground, specifiedColor=<rgba 1 0 1 1>}
      font {
       bForceMonospaced=false
       bMaterialIsStencil=false
       fontMaterial=null
       fontObject=&Roboto
       letterSpacing=0
       monospacedWidth=1
       size=12
       skewAmount=0
       typefaceFontName=Regular
       outlineSettings {
        bApplyOutlineToDropShadows=false
        bMiteredCorners=false
        bSeparateFillAlpha=false
        outlineMaterial=null
        outlineSize=0
        outlineColor=<rgba 0 0 0 1>
       }
      }
      highlightColor {colorUseRule=UseColor_Specified, specifiedColor=<rgba 0.02 0.3 0 1>}
      highlightShape {
       bIsDynamicallyLoaded=false
       drawAs=Box
       imageType=FullColor
       mirroring=NoMirror
       resourceName=../../../Engine/Content/Slate/Common/TextBlockHighlightShape.png
       resourceObject=null
       tiling=NoTile
       imageSize=<0 0>
       margin=<ltrb 0.375 0.375 0.375 0.375>
       outlineSettings {
        bUseBrushTransparency=false
        roundingType=HalfHeightRadius
        width=0
        color {colorUseRule=UseColor_Specified, specifiedColor=<rgba 0 0 0 0>}
        cornerRadii=<0 0 0 0>
       }
       tintColor {colorUseRule=UseColor_Specified, specifiedColor=<rgba 1 1 1 1>}
       uVRegion {bIsValid=false, max=<0 0>, min=<0 0>}
      }
      selectedBackgroundColor {colorUseRule=UseColor_Foreground, specifiedColor=<rgba 1 0 1 1>}
      shadowColorAndOpacity=<rgba 0 0 0 1>
      shadowOffset=<0 0>
      strikeBrush {
       bIsDynamicallyLoaded=false
       drawAs=Image
       imageType=NoImage
       mirroring=NoMirror
       resourceName=None
       resourceObject=null
       tiling=NoTile
       imageSize=<32 32>
       margin=<ltrb 0 0 0 0>
       outlineSettings {
        bUseBrushTransparency=false
        roundingType=HalfHeightRadius
        width=0
        color {colorUseRule=UseColor_Specified, specifiedColor=<rgba 0 0 0 0>}
        cornerRadii=<0 0 0 0>
       }
       tintColor {colorUseRule=UseColor_Specified, specifiedColor=<rgba 1 1 1 1>}
       uVRegion {bIsValid=false, max=<0 0>, min=<0 0>}
      }
      underlineBrush {
       bIsDynamicallyLoaded=false
       drawAs=Image
       imageType=NoImage
       mirroring=NoMirror
       resourceName=None
       resourceObject=null
       tiling=NoTile
       imageSize=<32 32>
       margin=<ltrb 0 0 0 0>
       outlineSettings {
        bUseBrushTransparency=false
        roundingType=HalfHeightRadius
        width=0
        color {colorUseRule=UseColor_Specified, specifiedColor=<rgba 0 0 0 0>}
        cornerRadii=<0 0 0 0>
       }
       tintColor {colorUseRule=UseColor_Specified, specifiedColor=<rgba 1 1 1 1>}
       uVRegion {bIsValid=false, max=<0 0>, min=<0 0>}
      }
     }
     vScrollBarPadding=<ltrb 0 0 0 0>
    }
   }
   slot {class=SizeBoxSlot}
  }
 ]
}
