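# Editor utility sub-widget for the audio tools Workflow Helper that lists asset types and provides one-click asset creation.
# Initialize fills a CommonListView with entry objects built from the helper data's AssetsToCreate map, returning false if it is empty.
# Generated entries get their button-press event bound.
# CreateAsset resolves the soft class and runs inline Python to create a uniquely named asset in the current Content Browser folder, then selects it.
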EditorUtilityWidgetBlueprint /FortTechAudioTools/Widgets/Workflow/SubWidgets/EUW_WH_AssetList
graphs [
 CreateAsset {nodes [
   n1 = CreateAsset!() -> (then, WorkflowEntryData:Struct WH Entry Widget Data Structure) {is=title K2Node_FunctionEntry}
   > Execute Python Script!(
    execute
    PythonScript="import unreal\r\n\r\ntry:\r\n    unreal.log(\"=== Starting asset creation ===\")\r\n    unreal.log(f\"Asset class type: {type(asset_class)}\")\r\n    unreal.log(f\"Asset class value: {asset_class}\")\r\n\r\n    ## Get the current Content Browser path\r\n    editor_util = unreal.EditorUtilityLibrary()\r\n    asset_path = editor_util.get_current_content_browser_path()\r\n\r\n    if not asset_path:\r\n        unreal.log_warning(\"Could not get current Content Browser path. Using /Game/\")\r\n        asset_path = \"/Game/\"\r\n    else:\r\n        unreal.log(f\"Current Content Browser path: {asset_path}\")\r\n\r\n    ## Get the class name\r\n    class_name = asset_class.get_name()\r\n    unreal.log(f\"Class name: {class_name}\")\r\n\r\n    ## Generate unique name with \"New\" + class name\r\n    asset_tools = unreal.AssetToolsHelpers.get_asset_tools()\r\n    base_path = f\"{asset_path}/New_{class_name}\"\r\n    unique_name = asset_tools.create_unique_asset_name(base_path, \"\")\r\n    package_path = unique_name[0]\r\n    asset_name = unique_name[1]\r\n\r\n    unreal.log(f\"Creating asset: {asset_name} at {asset_path}\")\r\n\r\n    ## Create asset\r\n    new_asset = asset_tools.create_asset(\r\n        asset_name,\r\n        asset_path,\r\n        asset_class,\r\n        None\r\n    )\r\n\r\n    if new_asset:\r\n        unreal.log(f\"SUCCESS: Created asset at {new_asset.get_path_name()}\")\r\n        unreal.EditorAssetLibrary.sync_browser_to_objects([new_asset.get_path_name()])\r\n    else:\r\n        unreal.log_error(\"FAILED: create_asset returned None\")\r\n\r\nexcept Exception as e:\r\n    unreal.log_error(f\"Exception occurred: {str(e)}\")\r\n    import traceback\r\n    unreal.log_error(traceback.format_exc())"
    asset_class:Object Class Reference=Resolve Soft Reference!(Input:Object Soft Class Reference=Break Struct WH Entry Widget Data!(Struct_WH_EntryWidget_Data:"Struct WH Entry Widget Data Structure (by ref)"=@n1) -> (
      bEnableButton_5_#1:bool=~
      ButtonText_2_#2:Text
      AssetClass_23_#3:Object Soft Class Reference
      Text_16_#4:Text
      TextTooltip_18_#5:Text
     ) {is=title K2Node_BreakStruct}.AssetClass_23_#3) -> (Output:Object Class Reference)
   ) -> (then, ReturnValue) {inputs [asset_class]}
  ]}
 EventGraph {nodes [
   n1 = "On Entry Generated (CommonListView_121)"!() -> (OutputDelegate, then, Widget:User Widget Object Reference) {is=title K2Node_ComponentBoundEvent}
   > n2 = Cast To WBP_WH_ListItem_Entry!(execute, Object=@n1.Widget) -> (then, CastFailed, AsWBP WH List Item Entry:WBP WH List Item Entry Object Reference) {is=title K2Node_DynamicCast}
   > Bind Event to On Button Pressed(self:WBP WH List Item Entry Object Reference=@n2, Delegate=Create Event()) {is=title K2Node_AddDelegate}
   Event PreConstruct()
   Event Construct()
   Event Tick()
  ]}
 Initialize {nodes [
   n1 = Initialize!() -> (then, WorkflowHelperData:BP Workflow Helper Data Object Reference) {is=title K2Node_FunctionEntry}
   > ClearListItems(self=Get CommonListView_121() -> (CommonListView_121:Common List View Object Reference))
   n2 = Get AssetsToCreate!(self:BP Workflow Helper Data Object Reference=@n1) -> (AssetsToCreate:Map of Object Soft Class References to Struct WH Entry Data Structures)
   > n3 = Branch(Condition=Is Empty!(TargetMap:Map of Object Soft Class References to Struct WH Entry Data Structures=@n2) -> (ReturnValue:bool) {is=title K2Node_CallFunction})
   > Return Node!(execute, bInitialized:bool=false)
   n4 = "For Each Loop (Map)"!(execute:Exec=@n3.else, MapPin:Map of Object Soft Class References to Struct WH Entry Data Structures=@n2, BreakPin:Exec) -> (then:Exec, KeyPin:Object Soft Class Reference, ValuePin:Struct WH Entry Data Structure, CompletedPin:Exec)
   n5 = Break Struct WH Entry Data!(Struct_WH_Entry_Data:"Struct WH Entry Data Structure (by ref)"=@n4.ValuePin) -> (Name_23_#2:Text, Tooltip_24_#5:Text) {is=title K2Node_BreakStruct}
   > n6 = Construct Object from Class!(
    execute
    Class=/FortTechAudioTools/Widgets/Workflow/SubWidgets/EntryWidgets/WBP_WH_ListItem_Entry.WBP_WH_ListItem_Entry_C
    self
    WorkflowEntryData:Struct WH Entry Widget Data Structure=Make Struct WH Entry Widget Data!(
     bEnableButton_5_#1:bool=True
     ButtonText_2_#2:Text=@n5
     AssetClass_23_#3:Object Soft Class Reference=@n4
     Text_16_#4:Text
     TextTooltip_18_#5:Text=@n5.Tooltip_24_#5
    ) -> (Struct_WH_EntryWidget_Data:Struct WH Entry Widget Data Structure) {is=title K2Node_MakeStruct}
   ) -> (then, ReturnValue:WBP WH List Item Entry Object Reference)
   > AddItem!(execute, self:List View Object Reference=Get CommonListView_121(), Item:Object Reference=@n6) -> (then)
   Return Node(execute=@n4.CompletedPin, bInitialized=true)
  ]}
]
variables []
widgets {
 info {rootWidgetClass=&CanvasPanel, widgetCount=3}
 widgets [
  CanvasPanel_0 {widgetClassPath=&CanvasPanel, widget=CanvasPanel_0}
  Border_0 {
   widgetClassPath=&Border
   parent=CanvasPanel_0
   widget=Border_0
   properties {
    background {
     bIsDynamicallyLoaded=false
     drawAs=Image
     imageType=NoImage
     mirroring=NoMirror
     resourceName=None
     resourceObject=null
     tiling=NoTile
     imageSize=<32 32>
     margin=<ltrb 0 0 0 0>
     outlineSettings {
      bUseBrushTransparency=false
      roundingType=HalfHeightRadius
      width=0
      color {colorUseRule=UseColor_Specified, specifiedColor=<rgba 0 0 0 0>}
      cornerRadii=<0 0 0 0>
     }
     tintColor {colorUseRule=UseColor_Specified, specifiedColor=<rgba 1 1 1 0>}
     uVRegion {bIsValid=false, max=<0 0>, min=<0 0>}
    }
    padding=<ltrb 0 5 5 5>
   }
   slot {
    class=CanvasPanelSlot
    properties {bAutoSize=true, layoutData {alignment=<0 0>, anchors {maximum=<1 1>, minimum=<0 0>}, offsets=<ltrb 0 0 0 0>}}
   }
  }
  CommonListView_121 {
   widgetClassPath=&CommonListView
   bIsVariable=true
   parent=Border_0
   widget=CommonListView_121
   properties {
    entryWidgetClass=&WBP_WH_ListItem_Entry_C
    verticalEntrySpacing=5
    scrollBarStyle {
     thickness=8
     draggedThumbImage {
      bIsDynamicallyLoaded=false
      drawAs=RoundedBox
      imageType=NoImage
      mirroring=NoMirror
      resourceName=None
      resourceObject=null
      tiling=NoTile
      imageSize=<0 0>
      margin=<ltrb 0 0 0 0>
      outlineSettings {
       bUseBrushTransparency=false
       roundingType=FixedRadius
       width=0
       color {colorUseRule=UseColor_Specified, specifiedColor=<rgba 0 0 0 0>}
       cornerRadii=<4 4 4 4>
      }
      tintColor {colorUseRule=UseColor_Specified, specifiedColor=<rgba 0.2159 0.2159 0.2159 1>}
      uVRegion {bIsValid=false, max=<0 0>, min=<0 0>}
     }
     horizontalBackgroundImage {
      bIsDynamicallyLoaded=false
      drawAs=NoDrawType
      imageType=NoImage
      mirroring=NoMirror
      resourceName=None
      resourceObject=null
      tiling=NoTile
      imageSize=<0 0>
      margin=<ltrb 0 0 0 0>
      outlineSettings {
       bUseBrushTransparency=false
       roundingType=HalfHeightRadius
       width=0
       color {colorUseRule=UseColor_Specified, specifiedColor=<rgba 0 0 0 0>}
       cornerRadii=<0 0 0 0>
      }
      tintColor {colorUseRule=UseColor_Specified, specifiedColor=<rgba 1 1 1 1>}
      uVRegion {bIsValid=false, max=<0 0>, min=<0 0>}
     }
     horizontalBottomSlotImage {
      bIsDynamicallyLoaded=false
      drawAs=NoDrawType
      imageType=NoImage
      mirroring=NoMirror
      resourceName=None
      resourceObject=null
      tiling=NoTile
      imageSize=<0 0>
      margin=<ltrb 0 0 0 0>
      outlineSettings {
       bUseBrushTransparency=false
       roundingType=HalfHeightRadius
       width=0
       color {colorUseRule=UseColor_Specified, specifiedColor=<rgba 0 0 0 0>}
       cornerRadii=<0 0 0 0>
      }
      tintColor {colorUseRule=UseColor_Specified, specifiedColor=<rgba 1 1 1 1>}
      uVRegion {bIsValid=false, max=<0 0>, min=<0 0>}
     }
     horizontalTopSlotImage {
      bIsDynamicallyLoaded=false
      drawAs=NoDrawType
      imageType=NoImage
      mirroring=NoMirror
      resourceName=None
      resourceObject=null
      tiling=NoTile
      imageSize=<0 0>
      margin=<ltrb 0 0 0 0>
      outlineSettings {
       bUseBrushTransparency=false
       roundingType=HalfHeightRadius
       width=0
       color {colorUseRule=UseColor_Specified, specifiedColor=<rgba 0 0 0 0>}
       cornerRadii=<0 0 0 0>
      }
      tintColor {colorUseRule=UseColor_Specified, specifiedColor=<rgba 1 1 1 1>}
      uVRegion {bIsValid=false, max=<0 0>, min=<0 0>}
     }
     hoveredThumbImage {
      bIsDynamicallyLoaded=false
      drawAs=RoundedBox
      imageType=NoImage
      mirroring=NoMirror
      resourceName=None
      resourceObject=null
      tiling=NoTile
      imageSize=<0 0>
      margin=<ltrb 0 0 0 0>
      outlineSettings {
       bUseBrushTransparency=false
       roundingType=FixedRadius
       width=0
       color {colorUseRule=UseColor_Specified, specifiedColor=<rgba 0 0 0 0>}
       cornerRadii=<4 4 4 4>
      }
      tintColor {colorUseRule=UseColor_Specified, specifiedColor=<rgba 0.2159 0.2159 0.2159 1>}
      uVRegion {bIsValid=false, max=<0 0>, min=<0 0>}
     }
     normalThumbImage {
      bIsDynamicallyLoaded=false
      drawAs=RoundedBox
      imageType=NoImage
      mirroring=NoMirror
      resourceName=None
      resourceObject=null
      tiling=NoTile
      imageSize=<0 0>
      margin=<ltrb 0 0 0 0>
      outlineSettings {
       bUseBrushTransparency=false
       roundingType=FixedRadius
       width=0
       color {colorUseRule=UseColor_Specified, specifiedColor=<rgba 0 0 0 0>}
       cornerRadii=<4 4 4 4>
      }
      tintColor {colorUseRule=UseColor_Specified, specifiedColor=<rgba 0.0953 0.0953 0.0953 1>}
      uVRegion {bIsValid=false, max=<0 0>, min=<0 0>}
     }
     verticalBackgroundImage {
      bIsDynamicallyLoaded=false
      drawAs=NoDrawType
      imageType=NoImage
      mirroring=NoMirror
      resourceName=None
      resourceObject=null
      tiling=NoTile
      imageSize=<0 0>
      margin=<ltrb 0 0 0 0>
      outlineSettings {
       bUseBrushTransparency=false
       roundingType=HalfHeightRadius
       width=0
       color {colorUseRule=UseColor_Specified, specifiedColor=<rgba 0 0 0 0>}
       cornerRadii=<0 0 0 0>
      }
      tintColor {colorUseRule=UseColor_Specified, specifiedColor=<rgba 1 1 1 1>}
      uVRegion {bIsValid=false, max=<0 0>, min=<0 0>}
     }
     verticalBottomSlotImage {
      bIsDynamicallyLoaded=false
      drawAs=NoDrawType
      imageType=NoImage
      mirroring=NoMirror
      resourceName=None
      resourceObject=null
      tiling=NoTile
      imageSize=<0 0>
      margin=<ltrb 0 0 0 0>
      outlineSettings {
       bUseBrushTransparency=false
       roundingType=HalfHeightRadius
       width=0
       color {colorUseRule=UseColor_Specified, specifiedColor=<rgba 0 0 0 0>}
       cornerRadii=<0 0 0 0>
      }
      tintColor {colorUseRule=UseColor_Specified, specifiedColor=<rgba 1 1 1 1>}
      uVRegion {bIsValid=false, max=<0 0>, min=<0 0>}
     }
     verticalTopSlotImage {
      bIsDynamicallyLoaded=false
      drawAs=NoDrawType
      imageType=NoImage
      mirroring=NoMirror
      resourceName=None
      resourceObject=null
      tiling=NoTile
      imageSize=<0 0>
      margin=<ltrb 0 0 0 0>
      outlineSettings {
       bUseBrushTransparency=false
       roundingType=HalfHeightRadius
       width=0
       color {colorUseRule=UseColor_Specified, specifiedColor=<rgba 0 0 0 0>}
       cornerRadii=<0 0 0 0>
      }
      tintColor {colorUseRule=UseColor_Specified, specifiedColor=<rgba 1 1 1 1>}
      uVRegion {bIsValid=false, max=<0 0>, min=<0 0>}
     }
    }
   }
   slot {class=BorderSlot, properties {padding=<ltrb 0 5 5 5>}}
  }
 ]
}
